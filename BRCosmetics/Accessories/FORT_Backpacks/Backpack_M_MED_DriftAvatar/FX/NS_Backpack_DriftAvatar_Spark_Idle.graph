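# Niagara system for the DriftAvatar backpack's idle spark effect: pink glows, a flickering light and electric bolt meshes, all attached to the backpack's skeletal mesh.
# Two sprite glows and a curve-and-waveform light scale with time of day.
# Shoulders_FX holds three bolts with glow, while Shoulders_FX002 bursts short-lived ones on random-length loops.
# Bolts derive a clamped XY bend from inherited movement versus sampled position, passing it to dynamic material parameters.
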NiagaraSystem /BRCosmetics/Accessories/FORT_Backpacks/Backpack_M_MED_DriftAvatar/FX/NS_Backpack_DriftAvatar_Spark_Idle
dependencies {
 usedDataInterfaces [&NiagaraDataInterfaceCurve, &NiagaraDataInterfaceArrayFloat3]
 usedDynamicInputs [
  &Multiply_VectorByFloat
  &SelectVectorFromArray
  &Boolean_NotOperation
 ]
 usedModules [
  &self:SystemSpawnScript.SystemScriptSource.SystemScriptGraph.NiagaraNodeAssignment_1.SetVariables_#1
  &SystemState
  &FN_Timer
  &self:SystemSpawnScript.SystemScriptSource.SystemScriptGraph.NiagaraNodeAssignment_0.SetVariables_#2
  &self:SystemSpawnScript.SystemScriptSource.SystemScriptGraph.NiagaraNodeAssignment_2.SetVariables_#3
 ]
}
emitters {
 Glow {
  emitterUpdateScript [
   &EmitterState(Life Cycle Mode=::NewEnumerator1, Loop Behavior=::NewEnumerator1, Loop Duration Mode=::NewEnumerator1, UseLoopDelay=false)
   &SpawnBurst_Instantaneous(Spawn Probability=HlslExpression() {hlslExpression=1.0}, Use Spawn Probability=false, UseLoopCountLimit=false)
  ]
  particleSpawnScript [&InitializeParticle(
    Color=<rgba 0.8632 0.013 0.4342 1>
    Sprite Size Mode=::NewEnumerator3
    Uniform Sprite Size=18
    UsePositionOffset=false
    Write Material Random=true
   )]
  particleUpdateScript [
   &ParticleState(Kill Particles When Lifetime Has Elapsed=false, Let Infinitely Lived Particles Die When Emitter Deactivates=true)
   &SkeletalMeshLocation(
    MaskSpawnGroup=false
    Mesh Sampling Type=::NewEnumerator5
    Sampled Position Offset=<0 0 6>
    Skeletal Mesh=$User.SkeletalMesh
    Socket Sampling Mode=::NewEnumerator5
   )
   &ScaleColor(
    Scale Alpha=&Lerp_Float(
     A=0
     B=&FNTimeOfDayScalar(Night Scale=0.5, Morning Scale=0.5, Evening Scale=0.5)
     Alpha=&FloatFromCurve(FloatCurve=DataInterface() {propertyValues {Curve {keys [
          {interpMode=RCIM_Cubic, tangentMode=RCTM_User, time=0, value=0}
          {arriveTangent=2.4173, interpMode=RCIM_Cubic, leaveTangent=2.4173, tangentMode=RCTM_User, time=1, value=1}
         ]}}})
    )
    ScaleA=true
    ScaleRGB=false
   )
   &CameraOffset()
  ]
  rendererClasses [&NiagaraSpriteRendererProperties]
  renderers [{rendererClass=&NiagaraSpriteRendererProperties}]
 }
 Glow001 {
  emitterUpdateScript [
   &EmitterState(Life Cycle Mode=::NewEnumerator1, Loop Behavior=::NewEnumerator1, Loop Duration Mode=::NewEnumerator1, UseLoopDelay=false)
   &SpawnBurst_Instantaneous(Spawn Probability=HlslExpression() {hlslExpression=1.0}, Use Spawn Probability=false, UseLoopCountLimit=false)
  ]
  particleSpawnScript [
   &InitializeParticle(
    Color=<rgba 0.8632 0.013 0.4342 1>
    Sprite Size Mode=::NewEnumerator3
    Uniform Sprite Size=12
    UsePositionOffset=false
    Write Material Random=true
   )
   &SpriteFacingAndAlignment(Sprite Alignment=<0 0 1>, Sprite Facing=<0 0 1>, WriteAlignment=false, WriteFacing=true)
  ]
  particleUpdateScript [
   &ParticleState(Kill Particles When Lifetime Has Elapsed=false, Let Infinitely Lived Particles Die When Emitter Deactivates=true)
   &SkeletalMeshLocation(
    MaskSpawnGroup=false
    Mesh Sampling Type=::NewEnumerator5
    Sampled Position Offset=<0 0 4.3>
    Skeletal Mesh=$User.SkeletalMesh
    Socket Sampling Mode=::NewEnumerator5
   )
   &ScaleColor(
    Scale Alpha=&Lerp_Float(
     A=0
     B=&FNTimeOfDayScalar(Night Scale=0.3, Morning Scale=0.5, Day Scale=0.5, Evening Scale=0.5)
     Alpha=&FloatFromCurve(FloatCurve=DataInterface() {propertyValues {Curve {keys [
          {interpMode=RCIM_Cubic, tangentMode=RCTM_User, time=0, value=0}
          {arriveTangent=2.4173, interpMode=RCIM_Cubic, leaveTangent=2.4173, tangentMode=RCTM_User, time=1, value=1}
         ]}}})
    )
    Scale RGB=&VectorFromFloat(Value=&FNTimeOfDayScalar(Night Scale=20, Morning Scale=40, Day Scale=50, Evening Scale=40))
    ScaleA=true
    ScaleRGB=true
   )
   &CameraOffset()
  ]
  rendererClasses [&NiagaraSpriteRendererProperties]
  renderers [{rendererClass=&NiagaraSpriteRendererProperties}]
 }
 Light {
  emitterSpawnScript [&FN_PreserveSkydiveUserVariables() {enabled=false, module=PreserveUserVariables}]
  emitterUpdateScript [
   &EmitterState(Life Cycle Mode=::NewEnumerator1, Loop Behavior=::NewEnumerator1, Loop Duration Mode=::NewEnumerator1, UseLoopDelay=false)
   &SpawnBurst_Instantaneous(Use Spawn Probability=false, UseLoopCountLimit=false)
  ]
  particleSpawnScript [&InitializeParticle(Lifetime=2, Sprite Size Mode=::NewEnumerator3, UsePositionOffset=false, Write Material Random=true)]
  particleUpdateScript [
   &ParticleState(
    Kill Particles When Lifetime Has Elapsed=false
    Let Infinitely Lived Particles Die When Emitter Deactivates=true
    Loop Particles Lifetime=true
   )
   &SkeletalMeshLocation(
    MaskSpawnGroup=false
    Mesh Sampling Type=::NewEnumerator5
    Sampled Position Offset=<0 0 10>
    Skeletal Mesh=$User.SkeletalMesh
    Socket Sampling Mode=::NewEnumerator5
   )
   /SetVariables(
    Particles.ColorLight:LinearColor=<rgba 0.8632 0.013 0.4342 1>
    Particles.LightIntensity:NiagaraFloat=&Multiply_Float(
     A=30
     B=&Multiply_Float(
      A=&Multiply_Float(
       A=&FNTimeOfDayScalar(Night Scale=0.3, Morning Scale=1, Evening Scale=1)
       B=&FloatFromCurve(FloatCurve=DataInterface() {propertyValues {Curve {keys [
            {arriveTangent=9.0725, interpMode=RCIM_Cubic, time=0, value=0.5}
            {arriveTangent=0.25, interpMode=RCIM_Cubic, leaveTangent=0.25, time=0.1408, value=0.9847}
            {arriveTangent=-0.1661, interpMode=RCIM_Cubic, leaveTangent=-0.1661, time=0.3, value=0.575}
            {arriveTangent=-0.5, interpMode=RCIM_Cubic, leaveTangent=-0.5, time=0.5, value=0.925}
            {arriveTangent=-0.0833, interpMode=RCIM_Cubic, leaveTangent=-0.0833, time=0.6, value=0.425}
            {arriveTangent=0.1875, interpMode=RCIM_Cubic, leaveTangent=0.1875, time=0.8, value=0.9}
            {interpMode=RCIM_Cubic, leaveTangent=-0.0405, time=1, value=0.5}
           ]}}})
      )
      B=&Waveform(
       Waveform Count=::NewEnumerator1
       "Waveform [1]"=::NewEnumerator0
       "[1] Frequency"=1
       "[1] Amplitude Min/Max"=<0.9 1>
       "[1] Scale"=0.5
       "[2] Average"=false
       "[2] Input Position"=0
       "[2] Phase"=$Emitter.Age
       "[2] Phase Scale"=1
       "[2] Compound Frequency"=2
       "[2] Frequency"=2
       "[2] Amplitude Min/Max"=<0 1>
       "[2] Scale"=0.3
       "[2] Bias"=0
       "[2] Exponent"=1
       "[2] Pulse Duration"=0.1
       "[2] Clip Triangle Min/Max"=<0 0>
       "[2] Random Seed"=0
      )
     )
    )
    Particles.LightIntensityComponent:NiagaraFloat=&Multiply_Float(
     A=50
     B=&Multiply_Float(
      A=&Multiply_Float(
       A=&FNTimeOfDayScalar(Night Scale=0.3, Morning Scale=1, Evening Scale=1)
       B=&FloatFromCurve(FloatCurve=DataInterface() {propertyValues {Curve {keys [
            {arriveTangent=9.0725, interpMode=RCIM_Cubic, time=0, value=0.5}
            {arriveTangent=0.25, interpMode=RCIM_Cubic, leaveTangent=0.25, time=0.1408, value=0.9847}
            {arriveTangent=-0.1661, interpMode=RCIM_Cubic, leaveTangent=-0.1661, time=0.3, value=0.575}
            {arriveTangent=-0.5, interpMode=RCIM_Cubic, leaveTangent=-0.5, time=0.5, value=0.925}
            {arriveTangent=-0.0833, interpMode=RCIM_Cubic, leaveTangent=-0.0833, time=0.6, value=0.425}
            {arriveTangent=0.1875, interpMode=RCIM_Cubic, leaveTangent=0.1875, time=0.8, value=0.9}
            {interpMode=RCIM_Cubic, leaveTangent=-0.0405, time=1, value=0.5}
           ]}}})
      )
      B=&Waveform(
       Waveform Count=::NewEnumerator1
       "Waveform [1]"=::NewEnumerator0
       "[1] Frequency"=1
       "[1] Amplitude Min/Max"=<0.9 1>
       "[1] Scale"=0.5
       "[2] Average"=false
       "[2] Input Position"=0
       "[2] Phase"=$Emitter.Age
       "[2] Phase Scale"=1
       "[2] Compound Frequency"=2
       "[2] Frequency"=2
       "[2] Amplitude Min/Max"=<0 1>
       "[2] Scale"=0.3
       "[2] Bias"=0
       "[2] Exponent"=1
       "[2] Pulse Duration"=0.1
       "[2] Clip Triangle Min/Max"=<0 0>
       "[2] Random Seed"=0
      )
     )
    )
   ) {isSetParameters=true}
   &FN_Light_Attributes(
    Color=<rgba 0.8632 0.013 0.4342 1>
    Create Color Attribute=true
    Create Exponent Attribute=true
    Create Position Attribute=true
    Create Radius Attribute=true
    Exponent=3
    Position=&Add_Vector(A=$Output.SkeletalMeshLocation.SampledPositionWithoutOffset, B=<0 0 15>)
    Radius=17
    Scale Color=$Particles.LightIntensity
   )
  ]
  rendererClasses [
   &NiagaraLightRendererProperties
   &NiagaraSpriteRendererProperties
   &NiagaraComponentRendererProperties
  ]
  renderers [
   {rendererClass=&NiagaraLightRendererProperties, rendererIndex=0}
   {rendererClass=&NiagaraSpriteRendererProperties, rendererIndex=1}
   {rendererClass=&NiagaraComponentRendererProperties, rendererIndex=2}
  ]
 }
 Shoulders_FX {
  emitterUpdateScript [
   &EmitterState(
    Inactive Response=::NewEnumerator2
    Life Cycle Mode=::NewEnumerator1
    Loop Behavior=::NewEnumerator1
    Loop Duration Mode=::NewEnumerator1
    UseLoopDelay=false
   )
   &SpawnBurst_Instantaneous(
    Spawn Count=3
    Spawn Probability=$User.ParticlesAmount
    Spawn Time=&RandomRangeFloat(Minimum=0, Maximum=0.1)
    Use Spawn Probability=true
    UseLoopCountLimit=false
   )
   /SetVariables(Emitter.DepthFade=35) {isSetParameters=true, module=SetVariables_#4}
  ]
  particleSpawnScript [
   &PartitionParticles(Number Of Partitions=3)
   /SetVariables(Particles.MeshIndex=$Particles.PartitionParticles.Partition, Particles.bIsTravelingBolt=true) {isSetParameters=true, module=SetVariables_#5}
   /SetVariables(Particles.SetWidth=4.5) {isSetParameters=true}
   /SetVariables(Particles.MeshIndex=5) {enabled=false, isSetParameters=true}
   &InitializeParticle(
    Color=&FNTimeOfDayColor(
     Night Color=<rgba 15 0.2256 7.5447 1>
     Morning Color=<rgba 30 0.4512 15.0895 1>
     Day Color=<rgba 20 0.3008 10.0597 1>
     Evening Color=<rgba 30 0.4512 15.0895 1>
    )
    Lifetime=$Emitter.CurrentLoopDuration
    Mesh Scale Mode=::NewEnumerator3
    Mesh Uniform Scale=&RandomRangeFloat(Minimum=1.1, Maximum=1.4)
    Sprite Rotation Mode=::NewEnumerator2
    Sprite Size Mode=::NewEnumerator3
    Uniform Sprite Size=2.2776
    UsePositionOffset=true
    Write Material Random=false
   )
   &DynamicMaterialParameters(
    Index 0 Param 2=3
    Index 0 Param 4=0.5
    Module.Param0WriteEnabled=true
    Module.Param1WriteEnabled=true
    Module.Param2WriteEnabled=true
    Module.Param3WriteEnabled=true
   )
   &InitialMeshOrientation(
    Mesh Orientation Mode=::NewEnumerator2
    Orientation Coordinate Space=::Local
    Orientation Vector=<0 0 1>
    Rotation Coordinate Space=::NewEnumerator2
    Use Rotation Vector=true
   )
  ]
  particleUpdateScript [
   &ParticleState(Kill Particles When Lifetime Has Elapsed=false, Let Infinitely Lived Particles Die When Emitter Deactivates=true)
   &CameraOffset(Camera Offset Amount=-1.5) {module=CameraOffset001}
   &InheritSourceMovement(
    Applied Position Falloff Distance=3
    Applied Position Scale=&FN_FrontEndVector3Modifier(Gameplay=<-0.75 -0.75 1>, Front End=<-1 -1 1>)
    ApplyForce=false
    ApplyPosition=true
    ApplyPositionFalloff=false
   )
   /SetVariables(Particles.PositionOrient:NiagaraPosition=&AddVectorToPosition(Position=$Particles.Position, Vector=<0 0 15>)) {isSetParameters=true}
   &SolveForcesAndVelocity(Clamp Velocity=false, Limit Acceleration=false)
   &SkeletalMeshLocation(
    MaskSpawnGroup=false
    Mesh Sampling Type=::NewEnumerator5
    Sampled Position Offset=&FN_FrontEndVector3Modifier(Gameplay=<0.2 0 4>, Front End=<0.2 0 2>)
    Sampled Position Offset Coordinate Space:ENiagaraSkelMeshTransforms=::NewEnumerator2
    Skeletal Mesh=$User.SkeletalMesh
    Socket Sampling Mode=::NewEnumerator5
   )
   /SetVariables(Particles.PositionBase:NiagaraPosition=&AddVectorToPosition(Position=$Particles.SkeletalMeshLocation.SampledPosition, Vector=<0 0 25>)) {isSetParameters=true}
   /SetVariables(Particles.BendPosition:Vector3f=&MakeVector(
     X=&ClampFloat(
      Float=&Subtract_Float(
       A=&MakeFloatFromVector(Vector=Linked() {linkedVariable {name=Particles.PositionOrient, type=NiagaraPosition}}, Channel=::NewEnumerator0)
       B=&MakeFloatFromVector(Vector=Linked() {linkedVariable {name=Particles.PositionBase, type=NiagaraPosition}}, Channel=::NewEnumerator0)
      )
      Min=-10
      Max=10
     )
     Y=&ClampFloat(
      Float=&Subtract_Float(
       A=&MakeFloatFromVector(Vector=Linked() {linkedVariable {name=Particles.PositionOrient, type=NiagaraPosition}}, Channel=::NewEnumerator1)
       B=&MakeFloatFromVector(Vector=Linked() {linkedVariable {name=Particles.PositionBase, type=NiagaraPosition}}, Channel=::NewEnumerator1)
      )
      Min=-10
      Max=10
     )
     Z=0
    )) {isSetParameters=true}
   &DynamicMaterialParameters(
    Index 0 Param 1=$Particles.SetWidth
    Index 0 Param 3=0.5
    Index 0 Param 4=4
    Index 3 Param 1=&Multiply_Float(A=&MakeFloatFromVector(Vector=$Particles.BendPosition, Channel=::NewEnumerator0), B=1)
    Index 3 Param 2=&Multiply_Float(A=&MakeFloatFromVector(Vector=$Particles.BendPosition, Channel=::NewEnumerator1), B=1)
    Module.Param0WriteEnabled=true
    Module.Param0WriteEnabled3=true
    Module.Param1WriteEnabled=false
    Module.Param1WriteEnabled3=true
    Module.Param2WriteEnabled=true
    Module.Param2WriteEnabled3=true
    Module.Param3WriteEnabled=true
    Module.Param3WriteEnabled3=true
    Write Parameter Index 3=true
   ) {module=DynamicMaterialParameters001}
   /SetVariables(
    Particles.CameraOffsetGlow:NiagaraFloat=&FN_FrontEndFloatModifier(Gameplay=-5.5, Front End=-2.5)
    Particles.ColorGlow:LinearColor=<rgba 0.401 0 0.1722 5>
    Particles.DynamicMaterialParameterGlow:Vector4f=&MakeVector4(X=&Add_Float(A=$Particles.SetWidth, B=8), Y=3, Z=0.5, W=4)
   ) {isSetParameters=true}
   /Game/Effects/Niagara/Modules/Curves/FN_HermiteSpline()
  ]
  rendererClasses [&NiagaraMeshRendererProperties, &NiagaraSpriteRendererProperties]
  renderers [
   {rendererClass=&NiagaraMeshRendererProperties, rendererIndex=0}
   {rendererClass=&NiagaraMeshRendererProperties, rendererIndex=1}
   {rendererClass=&NiagaraSpriteRendererProperties, rendererIndex=2}
   {rendererClass=&NiagaraSpriteRendererProperties, rendererIndex=3}
  ]
 }
 Shoulders_FX002 {
  emitterUpdateScript [
   &EmitterState(
    Inactive Response=::NewEnumerator2
    Life Cycle Mode=::NewEnumerator1
    Loop Duration=&RandomRangeFloat(Minimum=0.3, Maximum=0.8)
    Recalculate Duration Each Loop=true
    UseLoopDelay=false
   )
   &SpawnBurst_Instantaneous(
    Spawn Count=&RandomRangeInt(Minimum=1, Maximum=3)
    Spawn Probability=$User.ParticlesAmount
    Spawn Time=&RandomRangeFloat(Minimum=0, Maximum=0.1)
    Use Spawn Probability=true
    UseLoopCountLimit=false
   )
   /SetVariables(Emitter.DepthFade=35) {isSetParameters=true, module=SetVariables_#4}
  ]
  particleSpawnScript [
   /SetVariables(Particles.MeshIndex=&RandomRangeInt(Minimum=0, Maximum=2), Particles.bIsTravelingBolt=true) {isSetParameters=true, module=SetVariables_#5}
   &InitializeParticle(
    Color=&FNTimeOfDayColor(
     Night Color=<rgba 50 8.0729 20.3941 1>
     Morning Color=<rgba 50 8.0729 20.3941 1>
     Day Color=<rgba 50 8.0729 20.3941 1>
     Evening Color=<rgba 50 8.0729 20.3941 1>
    )
    Lifetime=&RandomRangeFloat(Minimum=0.42, Maximum=0.38)
    Mesh Scale Mode=::NewEnumerator3
    Mesh Uniform Scale=&RandomRangeFloat(Minimum=0.9, Maximum=1)
    Sprite Rotation Mode=::NewEnumerator2
    Sprite Size Mode=::NewEnumerator3
    UsePositionOffset=true
    Write Material Random=false
   )
   &CameraOffset(Camera Offset Amount=-2)
   &DynamicMaterialParameters(
    Index 0 Param 2=&RandomRangeFloat(Minimum=0.5, Maximum=0.75)
    Index 0 Param 4=0.5
    Module.Param0WriteEnabled=true
    Module.Param1WriteEnabled=true
    Module.Param2WriteEnabled=true
    Module.Param3WriteEnabled=true
   )
   &InitialMeshOrientation(Mesh Orientation Mode=::NewEnumerator2, Orientation Coordinate Space=::Local, Orientation Vector=<0 0 1>, Use Rotation Vector=true)
  ]
  particleUpdateScript [
   &ParticleState()
   &InheritSourceMovement(Applied Position Falloff Distance=3, Applied Position Scale=<0 0 1>, ApplyForce=false, ApplyPosition=true, ApplyPositionFalloff=false)
   /SetVariables(Particles.PositionOrient:NiagaraPosition=&AddVectorToPosition(Position=$Particles.Position, Vector=<0 0 15>)) {isSetParameters=true}
   &SolveForcesAndVelocity(Clamp Velocity=false, Limit Acceleration=false)
   &SkeletalMeshLocation(
    MaskSpawnGroup=false
    Mesh Sampling Type=::NewEnumerator5
    Sampled Position Offset=<0.5 0 3>
    Sampled Position Offset Coordinate Space:ENiagaraSkelMeshTransforms=::NewEnumerator2
    Skeletal Mesh=$User.SkeletalMesh
    Socket Sampling Mode=::NewEnumerator5
   )
   /SetVariables(Particles.PositionBase:NiagaraPosition=&AddVectorToPosition(Position=$Particles.SkeletalMeshLocation.SampledPosition, Vector=<0 0 25>)) {isSetParameters=true}
   /SetVariables(Particles.BendPosition:Vector3f=&MakeVector(
     X=&ClampFloat(
      Float=&Subtract_Float(
       A=&MakeFloatFromVector(Vector=Linked() {linkedVariable {name=Particles.PositionOrient, type=NiagaraPosition}}, Channel=::NewEnumerator0)
       B=&MakeFloatFromVector(Vector=Linked() {linkedVariable {name=Particles.PositionBase, type=NiagaraPosition}}, Channel=::NewEnumerator0)
      )
      Min=-10
      Max=10
     )
     Y=&ClampFloat(
      Float=&Subtract_Float(
       A=&MakeFloatFromVector(Vector=Linked() {linkedVariable {name=Particles.PositionOrient, type=NiagaraPosition}}, Channel=::NewEnumerator1)
       B=&MakeFloatFromVector(Vector=Linked() {linkedVariable {name=Particles.PositionBase, type=NiagaraPosition}}, Channel=::NewEnumerator1)
      )
      Min=-10
      Max=10
     )
     Z=0
    )) {isSetParameters=true}
   &DynamicMaterialParameters(
    Index 0 Param 1=&RandomRangeFloat(Minimum=2, Maximum=2.2)
    Index 0 Param 3=&Subtract_Float(A=1, B=$Particles.NormalizedAge)
    Index 0 Param 4=0.5
    Index 3 Param 1=&Multiply_Float(A=&MakeFloatFromVector(Vector=$Particles.BendPosition, Channel=::NewEnumerator0), B=1.1)
    Index 3 Param 2=&Multiply_Float(A=&MakeFloatFromVector(Vector=$Particles.BendPosition, Channel=::NewEnumerator1), B=1.1)
    Module.Param0WriteEnabled=true
    Module.Param0WriteEnabled3=true
    Module.Param1WriteEnabled=false
    Module.Param1WriteEnabled3=true
    Module.Param2WriteEnabled=true
    Module.Param2WriteEnabled3=true
    Module.Param3WriteEnabled=true
    Module.Param3WriteEnabled3=true
    Write Parameter Index 3=true
   ) {module=DynamicMaterialParameters001}
   &CameraOffset(Camera Offset Amount=-3) {module=CameraOffset001}
  ]
  rendererClasses [&NiagaraMeshRendererProperties]
  renderers [{rendererClass=&NiagaraMeshRendererProperties, rendererIndex=0}]
 }
}
summary {
 systemName=NS_Backpack_DriftAvatar_Spark_Idle
 emitters [
  {
   emitterName=Light
   rendererClasses [
    &NiagaraLightRendererProperties
    &NiagaraSpriteRendererProperties
    &NiagaraComponentRendererProperties
   ]
  }
  {emitterName=Glow, rendererClasses [&NiagaraSpriteRendererProperties]}
  {emitterName=Glow001, rendererClasses [&NiagaraSpriteRendererProperties]}
  {emitterName=Shoulders_FX002, rendererClasses [&NiagaraMeshRendererProperties]}
  {emitterName=Shoulders_FX, rendererClasses [&NiagaraMeshRendererProperties, &NiagaraSpriteRendererProperties]}
 ]
 userVariables [
  User.bIsFrontEnd {type=NiagaraBool}
  User.bIsFrontEndPreview {type=NiagaraBool}
  User.ColorSwitcher {type=NiagaraFloat, defaultValue=0, description=""}
  User.FlickerSpeed {type=NiagaraFloat, defaultValue=6.666, description=""}
  User.ForwardDot {type=NiagaraFloat, defaultValue=0, description=""}
  User.MaskParticlesAmount {type=NiagaraFloat, defaultValue=0, description=""}
  User.ParticlesAmount {type=NiagaraFloat, defaultValue=1, description=""}
  User.PowersActive {type=NiagaraBool}
  User.RibbonIDOffset {type=NiagaraInt32}
  User.SkeletalMesh {
   type=NiagaraDataInterfaceSkeletalMesh
   defaultValue {
    struct=NiagaraExt_VariableValue_DataInterface
    value {dataInterface=&self:NiagaraDataInterfaceSkeletalMesh_0}
   }
  }
 ]
}
